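# Default Lit material for the acorn prop mesh in the Sir Wolf backpack effects, enabled for use on Niagara mesh particles.
# It reuses the backpack's own diffuse, normal and mask textures through three plain texture samples.
# The graph has no math or parameter nodes, and no wiring from the samples to material outputs is shown.
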Material /Game/Accessories/FORT_Backpacks/Backpack_M_MED_SirWolf/FX/M_SirWolf_Acorn_Prop
graph {
 properties {
  bUsedWithNiagaraMeshParticles=true
  pixelDepthOffsetMode=PDOM_Legacy
  previewMesh=&/SM_SirWolf_Acorn_Prop
  usedShadingModels=Default Lit
 }
 nodes [
  TextureSample() {texture=/Game/Accessories/FORT_Backpacks/Backpack_M_MED_SirWolf/Textures/T_SirWolf_Backpack_D}
  TextureSample() {samplerType=SAMPLERTYPE_Normal, texture=/Game/Accessories/FORT_Backpacks/Backpack_M_MED_SirWolf/Textures/T_SirWolf_Backpack_N}
  TextureSample() {samplerType=SAMPLERTYPE_Masks, texture=/Game/Accessories/FORT_Backpacks/Backpack_M_MED_SirWolf/Textures/T_SirWolf_Backpack_S}
 ]
}
